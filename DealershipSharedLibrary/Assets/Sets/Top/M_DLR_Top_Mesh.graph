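# Masked lit material for the dealership "Top" mesh, with color zones, animated reflections and a dissolve effect.
# Rounded vertex colors form red, white, blue, glass and black masks, each setting color, metallic, roughness and specular; white gets procedural banding.
# Emissive combines cubemap reflection and a view-masked scrolling gradient.
# Noise blended with a UV gradient drives the dissolve's glowing blue edge, added to emissive through MF_FX_EmissiveAdjustments.
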Material /DealershipSharedLibrary/Assets/Sets/Top/M_DLR_Top_Mesh
graph {
 properties {
  bAutomaticallySetUsageInEditor=false
  bUsedWithMeshDeformer=true
  bUsedWithNiagaraMeshParticles=true
  bUsedWithSkeletalMesh=true
  blendMode=BLEND_Masked
  usedShadingModels=Default Lit
 }
 nodes [
  n1 = NamedRerouteUsage()
  n2 = ScalarParameter() {defaultValue=0.2, parameterName=EmissiveIntensity}
  n3 = ComponentMask(None=TextureCoordinate()) {g=true}
  n4 = ScalarParameter() {parameterName=RefletionDissolve}
  n5 = Multiply(A=NamedRerouteUsage(), B=Saturate(None=Subtract(A=@n3, B=@n4)))
  n6 = TextureCoordinate()
  n7 = Multiply(A=@n6, B=Constant2Vector() {g=0.5, r=1})
  n8 = Multiply(A=@n6, B=Constant2Vector() {g=0.4, r=1.5})
  n9 = Multiply(A=ComponentMask(None=TextureCoordinate()) {g=true})
  n10 = Subtract(A=@n9, B=OneMinus(None=@n9))
  n11 = LinearInterpolate(
   A=Multiply(
    A=TextureSample(UVs=@n7) {texture=/DealershipSharedLibrary/FX/Textures/Noise/T_VCVFX_CellNoiseSkins_A}.R
    B=TextureSample(UVs=@n8) {texture=/DealershipSharedLibrary/FX/Textures/Noise/T_VCVFX_Energy_Alpha}.R
   )
   B=OneMinus(None=Abs(None=@n10))
   Alpha=ScalarParameter() {defaultValue=0.4, parameterName=DissolveLerp}
  )
  n12 = ScalarParameter() {parameterName=Dissolveint}
  n13 = Saturate(None=Subtract(A=@n11, B=@n12) {constB=0.5})
  n14 = Subtract(A=@n11, B=Add(A=@n12) {constB=0.015})
  n15 = Multiply(A=Subtract(A=Ceil(None=@n13), B=Ceil(None=@n14)), B=Constant4Vector() {constant=<rgba 0.2 0.4 50 1>}.RGB)
  MaterialFunctionCall(Emissive=Add(A=Multiply(A=@n1, B=@n2), B=Add(A=@n5, B=@n15))) {materialFunction=&MF_FX_EmissiveAdjustments}
  n16 = ReflectionVectorWS()
  n17 = Saturate(None=DotProduct(A=CameraVectorWS(), B=Transform(None=@n16) {transformType=TRANSFORM_Camera}))
  n18 = Round(None=Subtract(A=Sine(None=@n17)) {constB=0})
  n19 = ReflectionVectorWS()
  Multiply(
   A=@n18
   B=Desaturation(
    None=TextureSampleParameterCube!(UVs=@n19, Apply View MipBias) {parameterName=CubeMap_02}.RGB
    Fraction=ScalarParameter() {defaultValue=0.5, parameterName=Fraction}
   )
  )
  n20 = Multiply(A=NamedRerouteUsage(), B=NamedRerouteUsage())
  Multiply(A=@n20, B=NamedRerouteUsage())
  n21 = VertexColor()
  NamedRerouteDeclaration(None=Round(None=@n21.R)) {name=R, nodeColor=<rgba 1 0 0 1>}
  NamedRerouteDeclaration(None=Round(None=@n21.G)) {name=G, nodeColor=<rgba 0 1 0 1>}
  n22 = Multiply(A=NamedRerouteUsage(), B=ScalarParameter() {defaultValue=0.5, parameterName=RedSpecular})
  n23 = Multiply(A=NamedRerouteUsage(), B=ScalarParameter() {defaultValue=0.5, parameterName=GlassSpecular})
  n24 = Add(
   A=Multiply(A=NamedRerouteUsage(), B=ScalarParameter() {defaultValue=0.3, parameterName=BlackSpecular})
   B=Multiply(A=NamedRerouteUsage(), B=ScalarParameter() {defaultValue=0.5, parameterName=BlueSpecular})
  )
  n25 = Multiply(A=NamedRerouteUsage(), B=ScalarParameter() {defaultValue=1, parameterName=WhiteSpecular})
  NamedRerouteDeclaration(None=Add(A=Add(A=@n22, B=@n23), B=Add(A=@n24, B=@n25))) {name=Specular, nodeColor=<rgba 0 1 0.8828 1>}
  n26 = Multiply(A=NamedRerouteUsage(), B=ScalarParameter() {defaultValue=0.3, parameterName=RedRoughness})
  n27 = Multiply(A=NamedRerouteUsage(), B=ScalarParameter() {parameterName=GlassRoughness})
  n28 = Add(
   A=Multiply(A=NamedRerouteUsage(), B=ScalarParameter() {defaultValue=0.4, parameterName=BlackRoughness})
   B=Multiply(A=NamedRerouteUsage(), B=ScalarParameter() {defaultValue=0.3, parameterName=BlueRoughness})
  )
  n29 = Multiply(A=NamedRerouteUsage(), B=ScalarParameter() {defaultValue=1, parameterName=WhiteRoughness})
  NamedRerouteDeclaration(None=Add(A=Add(A=@n26, B=@n27), B=Add(A=@n28, B=@n29))) {name=Roughness, nodeColor=<rgba 0 0.4141 1 1>}
  n30 = ReflectionVectorWS()
  n31 = Saturate(None=TextureSampleParameterCube(UVs=Transform(None=@n30)) {parameterName=CubeMap, samplerType=SAMPLERTYPE_LinearColor, texture=/NiagaraFluids/Textures/T_WaterReflection}.RGB)
  n32 = TextureCoordinate() {uTiling=0, vTiling=1.5}
  n33 = Multiply(A=Time(), B=ScalarParameter() {defaultValue=0.9, parameterName=MaskSpeed}) {constB=0.5}
  n34 = Multiply(
   A=@n18
   B=TextureSample(UVs=Add(A=@n32, B=@n33)) {texture=/DealershipSharedLibrary/FX/Textures/Gradient_Masks/T_Gradient_WheelWrapGradient_A}.R
  )
  NamedRerouteDeclaration(None=Multiply(A=Add(A=@n31, B=@n34), B=ScalarParameter() {defaultValue=3, parameterName=Metal reflection multiplier})) {name=Emissive, nodeColor=<rgba 0.125 1 0 1>}
  NamedRerouteDeclaration(None=Round(None=@n21.B)) {name=B, nodeColor=<rgba 0 0 1 1>}
  n35 = NamedRerouteUsage()
  n36 = NamedRerouteUsage()
  n37 = NamedRerouteDeclaration(None=Saturate(None=Subtract(A=@n35, B=@n36))) {name=Red, nodeColor=<rgba 0.5 0 1 1>}
  n38 = VectorParameter() {parameterName=Red, defaultValue=<rgba 0.75 0 0 0>}
  n39 = NamedRerouteDeclaration(None=Saturate(None=@n20)) {name=White, nodeColor=<rgba 0 1 0.1797 1>}
  n40 = ComponentMask(None=Add(A=TextureCoordinate(), B=Constant() {r=0.3})) {g=true}
  n41 = Round(None=ConstantBiasScale(None=Sine(None=@n40) {period=2}) {bias=-0.99, scale=70})
  n42 = ComponentMask(None=MaterialFunctionCall() {materialFunction=&BoundingBoxBased_0-1_UVW}.RGB) {g=true}
  n43 = Round(None=Subtract(A=Sine(None=@n42) {period=2}) {constB=0.25})
  n44 = LinearInterpolate(
   A=VectorParameter() {parameterName=White, defaultValue=<rgba 0.349 0.349 0.349 1>}.RGB
   Alpha=Multiply(A=Saturate(None=@n41), B=Saturate(None=@n43))
  ) {constB=0.15}
  n45 = NamedRerouteUsage()
  n46 = NamedRerouteUsage()
  n47 = Saturate(None=Subtract(A=Subtract(A=@n45, B=@n46), B=NamedRerouteUsage()))
  n48 = Add(
   A=Add(A=Multiply(A=@n37, B=@n38.RGB), B=Multiply(A=@n39, B=@n44))
   B=Multiply(
    A=NamedRerouteDeclaration(None=@n47) {name=Blue, nodeColor=<rgba 1 0 0.1406 1>}
    B=VectorParameter() {parameterName=Blue, defaultValue=<rgba 0.069 0.1984 0.3802 1>}.RGB
   )
  )
  n49 = Subtract(A=NamedRerouteUsage(), B=NamedRerouteUsage())
  n50 = Multiply(
   A=NamedRerouteDeclaration(None=Saturate(None=@n49)) {name=Glass, nodeColor=<rgba 0 0.4609 1 1>}
   B=VectorParameter() {parameterName=Glass, defaultValue=<rgba 0.1041 1 0.5049 2>}.RGB
  )
  n51 = NamedRerouteUsage()
  n52 = NamedRerouteUsage()
  n53 = NamedRerouteUsage()
  n54 = NamedRerouteUsage()
  n55 = Ceil(None=Add(A=Add(A=@n51, B=@n52), B=Add(A=@n53, B=@n54)))
  n56 = NamedRerouteDeclaration(None=Saturate(None=OneMinus(None=@n55))) {name=Black, nodeColor=<rgba 0.7812 1 0 1>}
  n57 = VectorParameter() {parameterName=Black, defaultValue=<rgba 0.03 0.03 0.03 0>}
  NamedRerouteDeclaration(None=Add(A=Add(A=@n48, B=@n50), B=Multiply(A=@n56, B=@n57.RGB))) {name=Color, nodeColor=<rgba 0.8984 0 1 1>}
  n58 = Multiply(A=NamedRerouteUsage(), B=ScalarParameter() {defaultValue=0.4, parameterName=RedMetallic})
  n59 = Multiply(A=NamedRerouteUsage(), B=ScalarParameter() {defaultValue=0.7, parameterName=GlassMetallic})
  n60 = Add(
   A=Multiply(A=NamedRerouteUsage(), B=ScalarParameter() {defaultValue=0.5, parameterName=BlackMetallic})
   B=Multiply(A=NamedRerouteUsage(), B=ScalarParameter() {defaultValue=0.8, parameterName=BlueMetallic})
  )
  n61 = Multiply(A=NamedRerouteUsage(), B=ScalarParameter() {defaultValue=0.5, parameterName=WhiteMetallic})
  NamedRerouteDeclaration(None=Add(A=Add(A=@n58, B=@n59), B=Add(A=@n60, B=@n61))) {name=Metalic, nodeColor=<rgba 0.4766 1 0 1>}
  NamedRerouteUsage()
  NamedRerouteUsage()
  NamedRerouteUsage()
 ]
}
